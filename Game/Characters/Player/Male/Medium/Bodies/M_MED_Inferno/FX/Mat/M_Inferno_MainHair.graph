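# Unlit translucent material for the Inferno character's flaming main hair, drawn on Niagara mesh particles.
# Gradient or curve-atlas color, contact tint and scrolling scan glow drive emissive; edge masks, textures, depth fade and optional dissolve shape opacity.
# Vertex offset thickens spline strands and adds flame lift, sine and random swings, and attraction rotations around texture-baked root pivots.
# Particle dynamic parameters drive motion and glow, with simpler deformation on low-quality, forward and mobile paths.
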Material /Game/Characters/Player/Male/Medium/Bodies/M_MED_Inferno/FX/Mat/M_Inferno_MainHair
graph {
 properties {
  bUsedWithNiagaraMeshParticles=true
  blendMode=BLEND_Translucent
  pixelDepthOffsetMode=PDOM_Legacy
  shadingModel=MSM_Unlit
  usedShadingModels=Unlit
 }
 nodes [
  n1 = VectorParameter() {group=ColorSetting, parameterName=SimpleGradientColor_LowAreaColor, defaultValue=<rgba 0.5 0 2 1>}
  n2 = VectorParameter() {group=ColorSetting, parameterName=SimpleGradientColor_HighAreaColor, defaultValue=<rgba 0 0.8333 5 1>}
  n3 = TextureCoordinate()
  n4 = Power(
   Base=MaterialFunctionCall(Float2=MaterialFunctionCall(XYZ=@n3) {materialFunction=&Swizzle}.YX) {materialFunction=&BreakOutFloat2Components}.R
   Exp=ScalarParameter() {defaultValue=3, group=ColorSetting, parameterName=SimpleGradientColor_LowAreaPower}
  )
  n5 = OneMinus(None=Power(Base=OneMinus(None=@n4), Exp=ScalarParameter() {defaultValue=1, group=ColorSetting, parameterName=SimpleGradientColor_HighAreaPower}))
  n6 = TextureCoordinate()
  n7 = CurveAtlasRowParameter(CurveTime=MaterialFunctionCall(Float2=MaterialFunctionCall(XYZ=@n6) {materialFunction=&Swizzle}.YX) {materialFunction=&BreakOutFloat2Components}.R) {
   atlas=&/CLA_InfernoCurve
   curve=&/CL_Inferno_BaseHair
   group=ColorSetting
   parameterName=CurveColor
  }
  n8 = DepthFade(FadeDistance=ScalarParameter() {group=ColorSetting, parameterName=ContactColorDistance}) {fadeDistanceDefault=2}
  n9 = LinearInterpolate(
   A=LinearInterpolate(A=LinearInterpolate(A=@n1.RGB, B=@n2.RGB, Alpha=@n5), B=@n7, Alpha=ScalarParameter() {group=ColorSetting, parameterName=UseCurveColor})
   B=VectorParameter() {group=ColorSetting, parameterName=ContactColor, defaultValue=<rgba 0 0 0 1>}.RGB
   Alpha=Power(Base=OneMinus(None=@n8), Exp=ScalarParameter() {defaultValue=2, group=ColorSetting, parameterName=ContactColorGradientPow})
  )
  n10 = NamedRerouteUsage()
  n11 = MaterialFunctionCall(Float2=TextureCoordinate()) {materialFunction=&BreakOutFloat2Components}
  n12 = DynamicParameter() {parameterIndex=2, defaultValue=<rgba 0 0 0 1>, paramNames [SizeScale, Scan, RootEmissiveBoost, Param4]}
  n13 = Add(A=OneMinus(None=@n11.G), B=Add(A=@n12.Scan, B=Constant() {r=0.5}))
  n14 = Multiply(
   A=MaterialFunctionCall!(TexCoord=AppendVector(A=@n11.R, B=@n13), ClampChannel=Constant() {r=1}) {materialFunction=/Game/Effects/Fort_Effects/MaterialFunctions/MF_FX_TilingMethod}.Result
   B=Sine(None=@n13)
  )
  n15 = Multiply(
   A=Power(Base=Saturate(None=@n14), Exp=ScalarParameter() {defaultValue=1.7, group=FlameUp, parameterName=ScanAreaPower})
   B=Multiply(
    A=VectorParameter() {group=Scan, parameterName=ScanColorBoost, defaultValue=<rgba 11 11 11 1>}.RGB
    B=ScalarParameter() {defaultValue=0.7, group=TODFix, parameterName=TODScanScale}
   )
  )
  n16 = Power(Base=@n11.G, Exp=ScalarParameter() {defaultValue=0.5, group=Scan, parameterName=ScanMaskPow})
  n17 = Multiply(
   A=LinearInterpolate(A=@n9, B=Add(A=@n10, B=@n9), Alpha=ScalarParameter() {group=Texture_A, parameterName=UseTextureABlend})
   B=Add(A=Multiply(A=@n15, B=@n16))
  )
  n18 = Multiply(
   A=Power(Base=Saturate(None=@n11.G), Exp=ScalarParameter() {defaultValue=9, group=FlameUp, parameterName=FlameUpMaskPow})
   B=ScalarParameter() {defaultValue=10, group=FlameUp, parameterName=FlameUpMaskMul}
  )
  n19 = ScalarParameter() {defaultValue=0.2, group=TODFix, parameterName=TODEmissiveBoostScale}
  n20 = Add(A=Multiply(A=@n12.RootEmissiveBoost, B=Multiply(A=@n18, B=@n19)), B=ParticleColor().R)
  n21 = MaterialFunctionCall(Emissive=Divide(A=Multiply(A=@n17, B=@n20)) {constB=1.977}) {materialFunction=&MF_FX_EmissiveAdjustments}
  n22 = ScalarParameter() {defaultValue=0.8, group=TODFix, parameterName=TODEmissiveScale}
  Add(
   A=StaticSwitchParameter(
    True=MaterialFunctionCall!(In, Preskinned Position, AO, Gmap, Skin, Toon Shading, Inkline) {materialFunction=/Game/Characters/Player/Male/Medium/Bodies/M_MED_LocalZilla/Materials/Super_Levels/SL24_Functions/MFC_S24_SuperLevel}.Boxes Mask
    False=Constant()
   ) {parameterName=UseSLBox}
   B=Multiply(
    A=Multiply(A=@n21.Result, B=@n22)
    B=ScalarParameter() {defaultValue=1, group=InfernoSpecialFix, parameterName=WhiteVersionSpecialFix_ColorScale}
   )
  )
  DynamicParameter() {parameterIndex=3, defaultValue=<rgba 0 1 1 1>, paramNames [WPO?, Timer, SwingWPOSTR, RandomSwingTimer]}
  n23 = NamedRerouteUsage()
  n24 = MaterialFunctionCall(XYZ=TextureCoordinate()) {materialFunction=&Swizzle}
  n25 = VectorParameter() {
   group=EdgeMask
   parameterName=EdgeMask_EdgeMask_InvertControl
   channelNames {a="", b=Power, g=Multiply, r=Offset}
   defaultValue=<rgba 0 10 1 1>
  }
  n26 = VectorParameter() {group=EdgeMask, parameterName=EdgeMask_EdgeMask_Control, channelNames {a="", b=Power, g=Multiply, r=Offset}, defaultValue=<rgba 0 1 2 1>}
  n27 = LinearInterpolate(
   A=MaterialFunctionCall!(
    UorV=MaterialFunctionCall(Float2=@n24.YX) {materialFunction=&BreakOutFloat2Components}.R
    Invert_EdgeControl=@n25
    NoInvert_EdgeControl=@n26
   ) {materialFunction=/Game/Effects/Niagara/MasterMaterials/Functions/Common/FXCommon_EdgeMask_Ex}.Mask
   B=@n7.A
   Alpha=ScalarParameter() {group=EdgeMask, parameterName=UseCurveAlpha}
  )
  n28 = NamedRerouteUsage()
  n29 = Multiply(
   A=LinearInterpolate(A=@n27, B=Multiply(A=@n28, B=@n27), Alpha=ScalarParameter() {group=Texture_A, parameterName=UseTextureABlend})
   B=NamedRerouteUsage()
  )
  n30 = DepthFade(Opacity=Saturate(None=Multiply(A=@n23, B=@n29)), FadeDistance=ScalarParameter() {defaultValue=3, group=DepthFade, parameterName=DFDistance}) {fadeDistanceDefault=2}
  n31 = TextureCoordinate()
  n32 = VectorParameter() {
   group=LastFixOpacityEdge
   parameterName=LastFixOpacityEdge_EdgeMask_InvertControl
   channelNames {a="", b=Power, g=Multiply, r=Offset}
   defaultValue=<rgba 0 1 1 1>
  }
  n33 = VectorParameter() {
   group=LastFixOpacityEdge
   parameterName=LastFixOpacityEdge_EdgeMask_Control
   channelNames {a="", b=Power, g=Multiply, r=Offset}
   defaultValue=<rgba 0 1 1 1>
  }
  n34 = LinearInterpolate(
   B=MaterialFunctionCall(
    UorV=MaterialFunctionCall(Float2=@n31) {materialFunction=&BreakOutFloat2Components}.G
    Invert_EdgeControl=@n32
    NoInvert_EdgeControl=@n33
   ) {materialFunction=&FXCommon_EdgeMask_Ex}.Mask
   Alpha=ScalarParameter() {parameterName=UseLastFix}
  ) {constA=1}
  n35 = TextureCoordinate()
  n36 = Multiply(
   A=Power(
    Base=MaterialFunctionCall(Float2=@n35) {materialFunction=&BreakOutFloat2Components}.G
    Exp=ScalarParameter() {defaultValue=1, group=DissolveMainT, parameterName=Dissolve_MaskP}
   )
   B=ScalarParameter() {group=DissolveMainT, parameterName=Dissolve_GradientStrength}
  )
  n37 = VectorParameter() {group=DissolveMainT, parameterName=DissolveMainT2_UVControl, channelNames {a=VPan, b=UPan, g=VScale, r=UScale}, defaultValue=<rgba 1 1 0 0>}
  n38 = VectorParameter() {group=DissolveMainT, parameterName=DissolveMainT_UVControl, channelNames {a=VPan, b=UPan, g=VScale, r=UScale}, defaultValue=<rgba 1 1 0 0>}
  n39 = Min(
   A=MaterialFunctionCall(
    Texture=TextureObjectParameter() {group=DissolveMainT, parameterName=DissolveMainT2_Texture, sortPriority=0}
    AdditionalUVOffset=NamedRerouteUsage()
    UVControl_RGB=@n37
    UVControl_A=@n37
   ) {materialFunction=&FXCommon_TextureSample}.R
   B=MaterialFunctionCall(
    Texture=TextureObjectParameter() {group=DissolveMainT, parameterName=DissolveMainT_Texture, sortPriority=0}
    AdditionalUVOffset=NamedRerouteUsage()
    UVControl_RGB=@n38
    UVControl_A=@n38
   ) {materialFunction=&FXCommon_TextureSample}.R
  )
  n40 = Multiply(
   A=Add(A=Add(A=@n36, B=@n39), B=ScalarParameter() {group=DissolveMainT, parameterName=Dissolve_Offset})
   B=ScalarParameter() {defaultValue=1, group=DissolveMainT, parameterName=Dissolve_Mul}
  )
  n41 = ScalarParameter() {group=DissolveMainT, parameterName=Dissolve_Contrast}
  n42 = Multiply(
   A=Saturate(None=Multiply(A=@n30, B=@n34))
   B=StaticSwitchParameter(
    True=MaterialFunctionCall(In=@n40, Contrast=@n41) {materialFunction=&CheapContrast}.Result
    False=Constant() {r=1}
   ) {parameterName=UseDissolve}
  )
  n43 = ScalarParameter() {defaultValue=1, group=InfernoSpecialFix, parameterName=WhiteVersionSpecialFix_OpacityScale}
  n44 = VertexColor()
  n45 = ScalarParameter() {group=InfernoSpecialFix, parameterName=WhiteVersionSpecialFix_OpacityVCScalePower}
  Multiply(
   A=Multiply(A=Multiply(A=@n42, B=@n43), B=Power(Base=@n44.G, Exp=@n45))
   B=LinearInterpolate(B=@n44.A, Alpha=ScalarParameter() {defaultValue=1, group=InfernoSpecialFix, parameterName=InfernoCoatVersion}) {constA=1}
  )
  n46 = TextureSampleParameter2D(UVs=TextureCoordinate() {coordinateIndex=1}) {
   group=WPO_BaseSetting
   parameterName=RootPosTexture
   samplerType=SAMPLERTYPE_LinearColor
   texture=&/T_HairThirdLayerPos
  }
  n47 = Add(A=Multiply(A=@n46.RGB) {constB=2}) {constB=-1}
  n48 = VectorParameter() {group=WPO_BaseSetting, parameterName=BoundSize, defaultValue=<rgba 25 25 25 1>}
  NamedRerouteDeclaration(None=TransformPosition(None=Multiply(A=@n47, B=@n48.RGB)) {transformSourceType=TRANSFORMPOSSOURCE_Instance, transformType=TRANSFORMPOSSOURCE_World}) {name=RootPosWS, nodeColor=<rgba 0.3828 0 1 1>}
  n49 = TextureCoordinate()
  n50 = Power(
   Base=OneMinus(None=MaterialFunctionCall(Float2=@n49) {materialFunction=&BreakOutFloat2Components}.G)
   Exp=ScalarParameter() {group=WPO_RandomSwing, parameterName=RandomSwingMaskPow}
  )
  n51 = DynamicParameter() {parameterIndex=3, defaultValue=<rgba 0 1 1 1>, paramNames [WPO?, Timer, SwingWPOSTR, RandomSwingTimer]}
  n52 = Add(A=Multiply(A=VertexColor().R) {constB=100}, B=@n51.RandomSwingTimer)
  n53 = ScalarParameter() {defaultValue=0.65, group=WPO_RandomSwing, parameterName=RandomSwingSpeed}
  n54 = Multiply(
   A=MaterialFunctionCall(Value=Multiply(A=@n52, B=@n53), -1 to 1=StaticBool() {value=true}) {materialFunction=&LinearSine}."Rounded Linear Sine "
   B=ScalarParameter() {defaultValue=0.015, group=WPO_RandomSwing, parameterName=RandomSwingStrength}
  )
  n55 = Normalize(VectorInput=VectorParameter() {group=WPO_RandomSwing, parameterName=RandomSwingAxis, defaultValue=<rgba 0 1 1 1>}.RGB)
  NamedRerouteDeclaration(None=MaterialFunctionCall!(
    RotationAngle=Multiply(A=@n50, B=@n54)
    NormalizedRotationAxis=Transform(None=@n55) {transformSourceType=TRANSFORMSOURCE_Instance}
    PivotPoint=NamedRerouteUsage()
    Position
    Normal
    TangentU
    TangentV
   ) {materialFunction=&/RotateVertexAboutAxis}.WPO) {name=WPO_RandomSwing, nodeColor=<rgba 0 0.1797 1 1>}
  n56 = CrossProduct(
   A=Subtract(A=CameraPositionWS(), B=WorldPosition().XYZ)
   B=MaterialFunctionCall() {materialFunction=&TangentBasis}.Y V
  )
  n57 = TextureCoordinate()
  n58 = Power(
   Base=OneMinus(None=MaterialFunctionCall(Float2=@n57) {materialFunction=&BreakOutFloat2Components}.G)
   Exp=ScalarParameter() {defaultValue=1, group=WPO_SineSwing, parameterName=SinSwingMaskPow}
  )
  n59 = Multiply(
   A=ScalarParameter() {defaultValue=1, group=WPO_SineSwing, parameterName=SinSwingPeriod}
   B=MaterialFunctionCall(Float2=TextureCoordinate()) {materialFunction=&BreakOutFloat2Components}.G
  )
  n60 = DynamicParameter() {parameterIndex=3, defaultValue=<rgba 0 1 1 1>, paramNames [WPO?, Timer, SwingWPOSTR, RandomSwingTimer]}
  n61 = Multiply(A=@n60.Timer, B=ScalarParameter() {defaultValue=2, group=WPO_SineSwing, parameterName=SinSwingTimeScale})
  n62 = Multiply(
   A=Sine(None=Add(A=@n59, B=@n61))
   B=Multiply(A=ScalarParameter() {defaultValue=2, group=WPO_SineSwing, parameterName=SinSwingStrength}, B=@n60.SwingWPOSTR)
  )
  NamedRerouteDeclaration(None=Multiply(A=Normalize(VectorInput=@n56), B=Multiply(A=@n58, B=@n62))) {name=WPO_SineSwing, nodeColor=<rgba 0 0.3672 1 1>}
  n63 = OneMinus(None=MaterialFunctionCall(Float2=TextureCoordinate()) {materialFunction=&BreakOutFloat2Components}.G)
  n64 = WorldPosition() {worldPositionShaderOffset=WPT_ExcludeAllShaderOffsets}
  n65 = NamedRerouteUsage()
  n66 = DynamicParameter() {parameterIndex=2, defaultValue=<rgba 0 0 0 1>, paramNames [SizeScale, Scan, RootEmissiveBoost, Param4]}
  n67 = Multiply(
   A=Normalize(VectorInput=Subtract(A=@n64.XYZ, B=@n65))
   B=LinearInterpolate(B=ScalarParameter() {defaultValue=6, group=WPO_FlameUp, parameterName=FlameUpStrength}, Alpha=@n66.SizeScale)
  )
  n68 = NamedRerouteUsage()
  n69 = NamedRerouteUsage()
  n70 = NamedRerouteUsage()
  n71 = Add(A=NamedRerouteUsage(), B=Add(A=NamedRerouteUsage(), B=Add(A=@n69, B=@n70)))
  n72 = Add(A=QualitySwitch(Default=ShadingPathSwitch(Default=@n71, Forward=@n70, Mobile=@n70), Low=@n70), B=NamedRerouteUsage())
  n73 = DynamicParameter() {parameterIndex=3, defaultValue=<rgba 0 1 1 1>, paramNames [WPO?, Timer, SwingWPOSTR, RandomSwingTimer]}
  NamedRerouteDeclaration(None=Add(A=Multiply(A=@n63, B=@n67), B=LinearInterpolate(A=@n68, B=@n72, Alpha=@n73.WPO?))) {name=WPOFinalResult, nodeColor=<rgba 1 0 0.7031 1>}
  n74 = NamedRerouteUsage()
  n75 = MaterialFunctionCall() {materialFunction=&TangentBasis}
  n76 = NamedRerouteUsage()
  n77 = LinearInterpolate(B=VertexColor().A, Alpha=ScalarParameter() {defaultValue=1, group=InfernoSpecialFix, parameterName=InfernoCoatVersion}) {constA=1}
  n78 = VertexColor()
  n79 = Multiply(A=Multiply(A=@n77, B=ScalarParameter() {defaultValue=5, group=WPO_SplineSetting, parameterName=SplineSetting_WidthB}), B=@n78.B)
  n80 = DynamicParameter() {paramNames [PointAttractionSwingX, PointAttractionSwingY, PointAttractionSwingZ, WidthScale]}
  n81 = ScalarParameter() {group=WPO_SplineSetting, parameterName=SplineSetting_UseDynamicInputControlTheWidth}
  n82 = Multiply(A=Multiply(A=@n77, B=ScalarParameter() {group=WPO_SplineSetting, parameterName=SplineSetting_WidthT}), B=@n78.B)
  n83 = WorldPosition() {worldPositionShaderOffset=WPT_ExcludeAllShaderOffsets}
  n84 = Add(A=Add(A=NamedRerouteUsage(), B=@n83.XYZ), B=NamedRerouteUsage())
  n85 = MaterialFunctionCall!(
   TangentU_WS=QualitySwitch(Default=ShadingPathSwitch(Default=@n74, Forward=@n75.X U, Mobile=@n75.X U), Low=@n75.X U)
   TangentV_WS=QualitySwitch(Default=ShadingPathSwitch(Default=@n76, Forward=@n75.Y V, Mobile=@n75.Y V), Low=@n75.Y V)
   WidthBase=LinearInterpolate(A=@n79, B=Multiply(A=@n79, B=@n80.WidthScale), Alpha=@n81)
   WidthTip=LinearInterpolate(A=@n82, B=Multiply(A=@n82, B=@n80.WidthScale), Alpha=@n81)
   WorldPosition=QualitySwitch(Default=ShadingPathSwitch(Default=@n84, Forward=@n83.XYZ, Mobile=@n83.XYZ), Low=@n83.XYZ)
   UVs for Projection
   UVs for Thickness
   UVs for texturing
   Expand U or V UV Channel=StaticBool() {value=true}
   DeriveNormalZ
   AdditionalNormal
   AngleCorrectedNormals
   FlattenPixelNormal
   UVs For NormalShading
  ) {materialFunction=/Game/Effects/Niagara/Modules/FollowingDeform/Material/SplineThickenWithTangentInput}
  NamedRerouteDeclaration(None=@n85.WorldPositionOffset) {name=SplineWPO, nodeColor=<rgba 1 0 0.9141 1>}
  n86 = VertexColor()
  NamedRerouteDeclaration(None=MaterialFunctionCall(X=Multiply(A=@n86.R) {constB=123.7587}, Y=Constant()) {materialFunction=&MakeFloat2}.Result) {name=UVRandomOffset, nodeColor=<rgba 1 0.8438 0 1>}
  n87 = TextureObjectParameter() {
   group=MainT
   parameterName=MainT_Texture
   samplerType=SAMPLERTYPE_Grayscale
   sortPriority=0
   texture=/Game/Effects/Fort_Effects/Textures/Smoke/T_Smoke_Trail_Soft_NoClamp
  }
  n88 = NamedRerouteUsage()
  n89 = VectorParameter() {group=MainT, parameterName=MainT_UVControl, channelNames {a=VPan, b=UPan, g=VScale, r=UScale}, defaultValue=<rgba 0.8 1 0.2 0>}
  n90 = MaterialFunctionCall(XYZ=TextureCoordinate()) {materialFunction=&Swizzle}
  n91 = VectorParameter() {group=MainT, parameterName=MainT_Refine, defaultValue=<rgba 1 0.1 0.5 0.5>}
  NamedRerouteDeclaration(None=MaterialFunctionCall(
    Texture=MaterialFunctionCall(Texture=@n87, AdditionalUVOffset=@n88, UVControl_RGB=@n89, UVControl_A=@n89, "[Override]UVInput"=@n90.YX) {materialFunction=&FXCommon_TextureSample}.R
    RGB=@n91.RGB
    A=@n91.RGB
   ) {materialFunction=&MF_FX_Refine}.Result) {name=MainTResult, nodeColor=<rgba 0 0.5547 1 1>}
  n92 = TextureObjectParameter() {
   group=Texture_A
   parameterName=Texture_A_Texture
   samplerType=SAMPLERTYPE_LinearColor
   sortPriority=0
   texture=/Game/Effects/Fort_Effects/Emotes/UltraLight/Texture/T_Plasma_A_C
  }
  n93 = VectorParameter() {group=Texture_A, parameterName=Texture_A_UVControl, channelNames {a=VPan, b=UPan, g=VScale, r=UScale}, defaultValue=<rgba 1 1 0 0>}
  n94 = MaterialFunctionCall(
   Texture=@n92
   AdditionalUVOffset=NamedRerouteUsage()
   UVControl_RGB=@n93
   UVControl_A=@n93
   "[Override]UVInput"=MaterialFunctionCall(XYZ=TextureCoordinate()) {materialFunction=&Swizzle}.YX
  ) {materialFunction=&FXCommon_TextureSample}
  n95 = VectorParameter() {group=Texture_A, parameterName=Texture_A_Ref}
  n96 = MaterialFunctionCall(XYZ=TextureCoordinate()) {materialFunction=&Swizzle}
  n97 = Power(
   Base=OneMinus(None=MaterialFunctionCall(Float2=@n96.YX) {materialFunction=&BreakOutFloat2Components}.R)
   Exp=ScalarParameter() {group=Texture_A, parameterName=Texture_A_MaskPow}
  ) {constExponent=0}
  NamedRerouteDeclaration(None=LinearInterpolate(
    A=MaterialFunctionCall(Texture=@n94.R, RGB=@n95.RGB, A=@n95.RGB) {materialFunction=&MF_FX_Refine}.Result
    Alpha=@n97
   )) {name=TextureA_Opacity, nodeColor=<rgba 0.6641 0 1 1>}
  n98 = MaterialFunctionCall(
   Hue Shift Percentage=ScalarParameter() {group=Texture_A, parameterName=TextureAHueShift}
   Texture=LinearInterpolate(A=@n94.RGB, Alpha=@n97) {constB=0}
  ) {materialFunction=&HueShift}
  n99 = ScalarParameter() {group=Texture_A, parameterName=TextureADesat}
  NamedRerouteDeclaration(None=Multiply(
    A=Desaturation(None=@n98.Result, Fraction=@n99)
    B=VectorParameter() {group=Texture_A, parameterName=TextureA_ColorMyl, defaultValue=<rgba 1 1 1 1>}.RGB
   )) {name=TextureA_Color, nodeColor=<rgba 0.6641 0 1 1>}
  n100 = Multiply(A=DotProduct(A=CameraVectorWS(), B=NamedRerouteUsage()), B=ScalarParameter() {defaultValue=4.5, group=Fresnel, parameterName=FresnelMul})
  NamedRerouteDeclaration(None=Power(Base=Saturate(None=@n100), Exp=ScalarParameter() {defaultValue=3, group=Fresnel, parameterName=FresnelPow}) {constExponent=1}) {name=FresnelOpacity, nodeColor=<rgba 1 0.7266 0 1>}
  n101 = DynamicParameter() {paramNames [PointAttractionSwingX, PointAttractionSwingY, PointAttractionSwingZ, WidthScale]}
  MaterialExpressionAppend3Vector_1 = MaterialXAppend3Vector!(A=@n101.PointAttractionSwingX, B=@n101.PointAttractionSwingX, C=@n101.PointAttractionSwingX)
  n102 = DotProduct(
   A=@MaterialExpressionAppend3Vector_1
   B=Transform(None=VectorParameter() {group=WPO_PointAttractions, parameterName=PointAttractionAxis, defaultValue=<rgba 0 0 1 1>}.RGB) {transformSourceType=TRANSFORMSOURCE_Instance}
  )
  n103 = ScalarParameter() {defaultValue=-0.002, group=WPO_PointAttractions, parameterName=PointAttractionSwingStr}
  n104 = OneMinus(None=MaterialFunctionCall(Float2=TextureCoordinate()) {materialFunction=&BreakOutFloat2Components}.G)
  n105 = ScalarParameter() {group=WPO_PointAttractions, parameterName=PointAttractionSwingMaskPow}
  n106 = MaterialFunctionCall(
   RotationAngle=Multiply(A=Multiply(A=@n102, B=@n103), B=Power(Base=@n104, Exp=@n105))
   NormalizedRotationAxis=Transform(None=VectorParameter() {group=WPO_PointAttractions, parameterName=PointAttractionRotationAxis, defaultValue=<rgba 1 0 0 1>}.RGB) {transformSourceType=TRANSFORMSOURCE_Instance}
   PivotPoint=NamedRerouteUsage()
   Position=Add(A=NamedRerouteUsage(), B=WorldPosition().XYZ)
  ) {materialFunction=&RotateVertexAboutAxis}
  NamedRerouteDeclaration(None=@n106.WPO) {name=WPO_DragLR, nodeColor=<rgba 1 0.7969 0 1>}
  NamedRerouteDeclaration(None=@n106.Normal) {name=DragLR_Normal, nodeColor=<rgba 1 0.2344 0 1>}
  NamedRerouteDeclaration(None=@n106.TangentU) {name=DragLR_TangentU, nodeColor=<rgba 1 0.2344 0 1>}
  NamedRerouteDeclaration(None=@n106.TangentV) {name=DragLR_TangentV, nodeColor=<rgba 1 0.2344 0 1>}
  NamedRerouteDeclaration(None=@n85.Normal) {name=SplineNormal, nodeColor=<rgba 0 0.5312 1 1>}
  n107 = DynamicParameter() {parameterIndex=1, defaultValue=<rgba 0 1 1 1>, paramNames [Shake,IA,IB,IC]}
  MaterialExpressionAppend3Vector_3 = MaterialXAppend3Vector(A=@n107.IA, B=@n107.IA, C=@n107.IA)
  n108 = Length(None=@MaterialExpressionAppend3Vector_3) {is=class MaterialExpressionLength}
  n109 = ScalarParameter() {defaultValue=0.001, group=WPO_PointAttractions, parameterName=PointAttraction2SwingStr}
  n110 = OneMinus(None=MaterialFunctionCall(Float2=TextureCoordinate()) {materialFunction=&BreakOutFloat2Components}.G)
  n111 = ScalarParameter() {defaultValue=1, group=WPO_PointAttractions, parameterName=PointAttraction2SwingMaskPow}
  n112 = Transform(None=Normalize(VectorInput=VectorParameter() {group=WPO_PointAttractions, parameterName=PointAttraction2Axis, defaultValue=<rgba 1 0 0 1>}.RGB)) {transformSourceType=TRANSFORMSOURCE_Instance}
  n113 = NamedRerouteUsage()
  n114 = Add(A=Add(A=NamedRerouteUsage(), B=WorldPosition().XYZ), B=NamedRerouteUsage())
  n115 = MaterialFunctionCall(
   RotationAngle=Multiply(A=Multiply(A=@n108, B=@n109), B=Power(Base=@n110, Exp=@n111))
   NormalizedRotationAxis=CrossProduct(A=@n112, B=Normalize(VectorInput=@MaterialExpressionAppend3Vector_3))
   PivotPoint=@n113
   Position=@n114
   Normal=NamedRerouteUsage()
   TangentU=NamedRerouteUsage()
   TangentV=NamedRerouteUsage()
  ) {materialFunction=&RotateVertexAboutAxis}
  n116 = VectorParameter() {group=WPO_PointAttractions, parameterName=PointAttraction3Axis, defaultValue=<rgba 0 0 1 1>}
  n117 = MaterialFunctionCall(
   RotationAngle=Multiply(
    A=DotProduct(A=@MaterialExpressionAppend3Vector_3, B=@n112)
    B=ScalarParameter() {defaultValue=0.001, group=WPO_PointAttractions, parameterName=PointAttraction3SwingStr}
   )
   NormalizedRotationAxis=Transform(None=Normalize(VectorInput=@n116.RGB)) {transformSourceType=TRANSFORMSOURCE_Instance}
   PivotPoint=@n113
   Position=Add(A=@n115.WPO, B=@n114)
   Normal=@n115.Normal
   TangentU=@n115.Normal
   TangentV=@n115.Normal
  ) {materialFunction=&RotateVertexAboutAxis}
  NamedRerouteDeclaration(None=Add(A=@n115.WPO, B=@n117.WPO)) {name=WPO_PointAttraction, nodeColor=<rgba 1 0.6563 0 1>}
  NamedRerouteDeclaration(None=@n117.TangentU) {name=PointAttraction_TangentU, nodeColor=<rgba 1 0 0.9375 1>}
  NamedRerouteDeclaration(None=@n117.TangentV) {name=PointAttraction_TangentV, nodeColor=<rgba 1 0 0.9375 1>}
  NamedRerouteUsage()
  Time()
 ]
}
